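# Translucent Niagara sprite material that draws an animated solar flare on a flat card, filed under the Zeropoint skydiving contrail effects.
# Radial UVs drive panning magnetic waves, noise-distorted smoke and tiled streaks, with the particle's Random dynamic parameter offsetting waves and distortion noise.
# A square edge mask and inverted sphere mask shape the combined intensity, which blends two colors and multiplies particle alpha.
# Wave RGB is added on top of the blended color.
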Material /BRCosmetics/Athena/Effects/Materials/Skydiving/P_Contrail_Zeropoint/M_SolarFlare_FlatCard
graph {
 properties {bUsedWithNiagaraSprites=true, blendMode=BLEND_Translucent, pixelDepthOffsetMode=PDOM_Legacy, usedShadingModels=Default Lit}
 nodes [
  n1 = Multiply(
   A=Constant2Vector() {g=0.2, r=0.3}
   B=MaterialFunctionCall(Swizzle Coordinate Output=StaticBool() {value=true}) {materialFunction=&VectorToRadialValue}.Radial Coordinates
  ) {constA=0.3}
  n2 = DynamicParameter() {paramNames [Random,Param2,Param3,Param4]}
  n3 = AppendVector(A=@n2.Random, B=Constant())
  n4 = TextureSample(UVs=Panner(Coordinate=Add(A=@n1, B=@n3)) {speedX=-0.1, speedY=0.05}) {texture=/Game/Athena/Prototype/Blueprints/Sneezy/Textures/T_Sneezy_MagneticWaves}
  n5 = StaticBool() {value=true}
  n6 = Add(A=@n2.Random, B=Time())
  n7 = LinearInterpolate(Alpha=TextureSample(UVs=Panner(Time=@n6) {speedX=-0.6}) {samplerType=SAMPLERTYPE_Grayscale, texture=&LowResBlurredNoise}.G) {constA=-1}
  n8 = ScalarParameter() {defaultValue=0.1, parameterName=DistAmnt}
  n9 = TextureSample(UVs=Add(
    A=MaterialFunctionCall(Swizzle Coordinate Output=@n5) {materialFunction=&VectorToRadialValue}.Radial Coordinates
    B=Multiply(A=@n7, B=@n8)
   )) {texture=/Game/Effects/Fort_Effects/Textures/T_Night_Smoke_Blur}
  n10 = Multiply(A=@n4.G, B=ScalarParameter() {defaultValue=0.5, parameterName=CoorAmnt})
  n11 = Multiply(
   A=AppendVector(A=ScalarParameter() {defaultValue=3, parameterName=StreaksTiles}, B=Constant() {r=0.5})
   B=MaterialFunctionCall(Swizzle Coordinate Output=StaticBool()) {materialFunction=&VectorToRadialValue}.Radial Coordinates
  ) {constA=5}
  n12 = Multiply(
   A=@n9.G
   B=TextureSample(UVs=Panner(Coordinate=@n11) {speedX=-0.1, speedY=1}) {texture=/Game/Effects/Fort_Effects/Textures/Tile/T_Inky_Smoke_Tile2}.G
  )
  n13 = Add(A=Multiply(A=ScalarParameter() {defaultValue=0.6, parameterName=ExtraAdd}, B=@n9.G), B=Multiply(A=@n9.G, B=Add(A=@n10, B=@n12)))
  n14 = MaterialFunctionCall!(UVs, Mult=ScalarParameter() {defaultValue=5, parameterName=Mult}, Exp=ScalarParameter() {defaultValue=3, parameterName=Exp}) {materialFunction=/Game/Effects/Fort_Effects/MaterialFunctions/MF_SquareEdgeMask}
  n15 = SphereMask(A=TextureCoordinate(), B=Constant() {r=0.5}, Radius=Constant() {r=0.8}) {hardnessPercent=20}
  n16 = Multiply(A=ScalarParameter() {defaultValue=4, parameterName=bright}, B=Multiply(A=Multiply(A=@n13, B=@n14.SquareMask), B=OneMinus(None=@n15)))
  n17 = Add(
   A=@n4.RGB
   B=LinearInterpolate(
    A=VectorParameter() {parameterName=Color2, defaultValue=<rgba 0.0002 0.1793 1 0>}.RGB
    B=VectorParameter() {parameterName=Color, defaultValue=<rgba 0.5729 0.8385 5 0>}.RGB
    Alpha=Saturate(None=@n16)
   )
  )
  n18 = ParticleColor()
  Add(A=@n17, B=@n18.RGB)
  MaterialFunctionCall(Emissive=@n17) {materialFunction=&MF_FX_EmissiveAdjustments}
  Multiply(A=@n18.A, B=@n16)
  TextureCoordinate()
 ]
}
