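# Unlit translucent Niagara sprite material for the glowing core of the Zero Point skydiving contrail.
# Layered sine waves over time drive a flicker in the radius and brightness of procedural sphere masks, which blend two tints into emissive through an adjustment function.
# An outer sphere mask shapes opacity, multiplied by soft depth fading and an inverted distance fade against scene geometry.
# Mobile and low quality skip both depth fades.
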Material /BRCosmetics/Athena/Effects/Materials/Skydiving/P_Contrail_Zeropoint/M_ZeroPoint_Core_Contrails
graph {
 properties {bUsedWithNiagaraSprites=true, blendMode=BLEND_Translucent, pixelDepthOffsetMode=PDOM_Legacy, shadingModel=MSM_Unlit, usedShadingModels=Unlit}
 nodes [
  n1 = ScalarParameter() {defaultValue=2, parameterName=InnerMult}
  n2 = Multiply(A=Time(), B=ScalarParameter() {defaultValue=0.8, parameterName=FlickerSpeed})
  n3 = Multiply(B=Sine(None=@n2) {period=0.55})
  n4 = Multiply(A=Sine(None=@n2) {period=0.11}) {constB=0.2}
  n5 = Sine(None=@n2) {period=0.175}
  n6 = Sine(None=@n2) {period=3.25}
  n7 = Sine(None=@n2) {period=2.325}
  n8 = Multiply(A=Add(A=Add(A=@n3, B=@n4), B=Multiply(A=@n5) {constB=0.3}), B=ConstantBiasScale(None=Multiply(A=@n6, B=@n7)))
  n9 = SphereMask(
   A=TextureCoordinate()
   B=Constant() {r=0.5}
   Radius=LinearInterpolate(A=@n1, B=ScalarParameter() {defaultValue=0.5, parameterName=RadiusInner}, Alpha=@n8)
   Hardness=ScalarParameter() {parameterName=HardnessInner}
  )
  n10 = TextureCoordinate()
  n11 = Constant() {r=0.5}
  n12 = ScalarParameter() {defaultValue=0.5, parameterName=RadiusInnerB}
  n13 = ScalarParameter() {parameterName=HardnessInnerB}
  n14 = Multiply(
   A=LinearInterpolate(A=@n1, B=ScalarParameter() {defaultValue=5, parameterName=FlickerMult}, Alpha=@n8)
   B=Multiply(A=SphereMask(A=@n10, B=@n11, Radius=@n12, Hardness=@n13), B=ScalarParameter() {defaultValue=2, parameterName=InnerMultB})
  )
  MaterialFunctionCall(Emissive=LinearInterpolate(
    A=VectorParameter() {parameterName=TintA, defaultValue=<rgba 0.0393 0 0.5625 1>}.RGB
    B=VectorParameter() {parameterName=TintB, defaultValue=<rgba 2 0.4 3 1>}.RGB
    Alpha=Add(A=@n9, B=@n14)
   )) {materialFunction=&MF_FX_EmissiveAdjustments}
  n15 = Constant() {r=0.5}
  n16 = OneMinus(None=DepthFade(FadeDistance=ScalarParameter() {defaultValue=3000, parameterName=DecalDist}))
  n17 = ScalarParameter() {defaultValue=10, parameterName=DFDist}
  n18 = Constant() {r=1}
  n19 = ShadingPathSwitch(Default=Multiply(A=Saturate(None=@n16), B=DepthFade(FadeDistance=@n17)), Mobile=@n18) {desc=Migrated from FeatureLevelSwitch}
  Multiply(
   A=SphereMask(A=TextureCoordinate(), B=@n15, Radius=@n15, Hardness=ScalarParameter() {parameterName=HardnessOuter})
   B=QualitySwitch(Default=@n19, Low=@n18, High=@n19, Medium=@n19)
  )
 ]
}
