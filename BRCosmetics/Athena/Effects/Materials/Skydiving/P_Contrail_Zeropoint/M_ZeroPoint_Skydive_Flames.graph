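# Additive unlit Niagara sprite material for flames in the Zero Point skydiving contrail.
# Panning blurred noise distorts both a 2x2 inky smoke flipbook, phased by a Rand dynamic parameter, and an edge-masked turbulence layer that boosts brightness.
# A curve atlas colors the combined mask, then turbulence tint and particle color shape the emissive.
# Opacity follows mask and particle alpha, with near and far depth fades bypassed on mobile and low quality.
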Material /BRCosmetics/Athena/Effects/Materials/Skydiving/P_Contrail_Zeropoint/M_ZeroPoint_Skydive_Flames
graph {
 properties {bUsedWithNiagaraSprites=true, blendMode=BLEND_Additive, pixelDepthOffsetMode=PDOM_Legacy, shadingModel=MSM_Unlit, usedShadingModels=Unlit}
 nodes [
  Constant()
  n1 = ParticleColor()
  n2 = TextureCoordinate()
  n3 = VectorParameter() {parameterName=NoiseTileSPeed, defaultValue=<rgba 4 2 0 2>}
  n4 = ComponentMask(None=@n3.RGB) {g=true, r=true}
  n5 = TextureSample(UVs=Panner(Coordinate=Multiply(A=@n2, B=@n4), Speed=AppendVector(A=@n3.B, B=@n3.B)) {speedY=1}) {samplerType=SAMPLERTYPE_Grayscale, texture=&LowResBlurredNoise}
  n6 = Multiply(A=ComponentMask(None=@n5.RGB) {g=true, r=true}, B=ScalarParameter() {defaultValue=0.025, parameterName=DistAmnt})
  n7 = Panner(Coordinate=Add(A=@n6, B=TextureCoordinate())) {speedX=0.05, speedY=0.2}
  n8 = Multiply(
   A=MaterialFunctionCall!(UVs, Mult, Exp) {materialFunction=/Game/Effects/Fort_Effects/MaterialFunctions/MF_SquareEdgeMask}.SquareMask
   B=MaterialFunctionCall(
    In=TextureSample(UVs=@n7) {texture=/Game/Global/Textures/Default/Caustics/T_Turbulence_Seamless}.G
    Contrast=ScalarParameter() {defaultValue=0.06, parameterName=NoiseContrast}
   ) {materialFunction=&CheapContrast}.Result
  )
  n9 = Multiply(A=VectorParameter() {parameterName=NoiseColor, defaultValue=<rgba 0 0.1 3 1>}.RGB, B=@n8)
  n10 = Add(A=@n5.G, B=ScalarParameter() {defaultValue=0.6, parameterName=AddNoise}) {constB=0.8}
  n11 = DynamicParameter() {defaultValue=<rgba 0.3 0 0 0>, paramNames [Rand, "", Param3, Param4]}
  n12 = Constant() {r=2}
  n13 = TextureCoordinate()
  n14 = TextureSample(UVs=Add(
    A=@n6
    B=MaterialFunctionCall("Animation  Phase (0-1)"=@n11.Rand, Number of Rows=@n12, Number of Columns=@n12, UVs=@n13) {materialFunction=&FlipBook}.UVs
   )) {texture=/BRCosmetics/Athena/Effects/Textures/T_InkySmoke_Flipbook}
  n15 = Saturate(None=Multiply(A=Add(A=@n10, B=@n14.G), B=@n14.G))
  n16 = Multiply(A=ScalarParameter() {defaultValue=12, parameterName=NoiseBrightness}, B=@n8)
  n17 = Saturate(None=Add(A=@n15, B=Multiply(A=@n15, B=@n16)))
  n18 = CurveAtlasRowParameter(CurveTime=@n17) {
   atlas=&AthenaCurveAtlas01
   curve=/Game/Athena/Environments/Nexus/Materials/FXshader/Curve_Color
   defaultValue=78
   parameterName=ColorCurve
  }
  MaterialFunctionCall(Emissive=Multiply(A=@n1.RGB, B=Add(A=@n9, B=@n18))) {materialFunction=&MF_FX_EmissiveAdjustments}
  n19 = ScalarParameter() {defaultValue=1, parameterName=Opac}
  n20 = ScalarParameter() {defaultValue=5, parameterName=DF_Dist}
  n21 = OneMinus(None=DepthFade(FadeDistance=ScalarParameter() {defaultValue=500000, parameterName=DecalDist}))
  n22 = Constant() {r=1}
  n23 = ShadingPathSwitch(Default=Multiply(A=DepthFade(FadeDistance=@n20), B=Saturate(None=@n21)), Mobile=@n22) {desc=Migrated from FeatureLevelSwitch}
  Multiply(A=Multiply(A=@n1.A, B=Multiply(A=@n19, B=@n17)), B=QualitySwitch(Default=@n23, Low=@n22, High=@n23, Medium=@n23))
 ]
}
